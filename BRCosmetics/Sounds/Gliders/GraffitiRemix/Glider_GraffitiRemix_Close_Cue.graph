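# Sound cue for the close sound of the Graffiti Remix glider cosmetic in Battle Royale.
# A single wave player feeds the close sound through a Modulator left at its default settings, which gives slight random pitch and volume variation.
# Playback uses the shared glider open/close attenuation and concurrency, the gliders landing sound class and the cosmetic glider-close property sheet, at reduced volume.
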SoundCue /BRCosmetics/Sounds/Gliders/GraffitiRemix/Glider_GraffitiRemix_Close_Cue
root=Modulator_0
nodes [
 {id=Modulator_0, children [WavePlayer_1]}
 {id=WavePlayer_1, properties {soundWaveAssetPtr=&/Athena_Glider_GraffitiRemix_Close}}
]
properties {
 attenuationSettings=/Game/Sounds/Attenuation/Fort_Player_Glider_OpenClose_Athena_Att
 audioPropertiesSheet=/BRCosmetics/Sounds/Settings/PropertySheets/Glider/APS_Cosmetic_Glider_Close
 maxDistance=10000
 soundClassObject=&Fort_Gliders_Land
 volumeMultiplier=0.35
 concurrencySet [&Athena_GliderOpenClose_Conc]
}
