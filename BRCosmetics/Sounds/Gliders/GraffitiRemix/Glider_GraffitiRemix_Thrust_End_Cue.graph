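# Sound cue for the thrust-end sound of the GraffitiRemix cosmetic glider.
# An equally weighted random node picks one of three thrust waves, then a modulator lowers its pitch within a narrow random range.
# The cue plays at reduced volume with glider thrust attenuation and concurrency, a glider landing sound class, and a glider loop audio properties sheet.
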SoundCue /BRCosmetics/Sounds/Gliders/GraffitiRemix/Glider_GraffitiRemix_Thrust_End_Cue
root=Modulator_0
nodes [
 {id=Modulator_0, children [Random_1], properties {pitchMax=0.85, pitchMin=0.8, volumeMin=1}}
 {id=Random_1, children [WavePlayer_1,WavePlayer_4,WavePlayer_5], properties {weights [1,1,1]}}
 {id=WavePlayer_1, properties {soundWaveAssetPtr=&/Athena_Glider_GraffitiRemix_Thrust_02}}
 {id=WavePlayer_4, properties {soundWaveAssetPtr=&/Athena_Glider_GraffitiRemix_Thrust_03}}
 {id=WavePlayer_5, properties {soundWaveAssetPtr=&/Athena_Glider_GraffitiRemix_Thrust_01}}
]
properties {
 attenuationSettings=&Fort_Player_Glider_Thrust_Athena_Att
 audioPropertiesSheet=&APS_Cosmetic_Glider_Loop
 maxDistance=12500
 soundClassObject=&Fort_Gliders_Land
 volumeMultiplier=0.5
 concurrencySet [&Athena_Glider_Thrust_Conc]
}
